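# Physics asset for the Sydney_01 male medium head cosmetic, a "zom" variant that covers only the jaw.
# It has two small capsule bodies with zero length, effectively spheres, on the C_jaw and faceAttach bones.
# A single constraint joins the jaw to faceAttach, with the second swing axis locked and narrow angle values set for swing and twist.
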PhysicsAsset /BRCosmetics/Characters/Player/Male/Medium/Heads/M_MED_BLK_Sydney_Head_01/Meshes/M_MED_BLK_Sydney_01_Physics_zom
bodies {C_jaw [{shapeType=Capsule, length=0, radius=1, center=<0 0.18 0.1>}], faceAttach [{shapeType=Capsule, length=0, radius=1}]}
constraints [{bone1Name=C_jaw, bone2Name=faceAttach, swing1LimitDegrees=10, swing2LimitDegrees=5, swing2Motion=Locked, twistLimitDegrees=22.5}]
